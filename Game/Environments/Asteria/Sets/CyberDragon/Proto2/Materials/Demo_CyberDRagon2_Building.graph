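# Placeholder-style Default Lit material, named for the buildings in the second CyberDragon prototype demo set of the Asteria environment.
# The graph holds only three scalar constants, one slightly above zero and two at default, with no textures, parameters or visible links to material outputs.
# It is flagged for use on Nanite meshes and uses legacy pixel depth offset mode.
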Material /Game/Environments/Asteria/Sets/CyberDragon/Proto2/Materials/Demo_CyberDRagon2_Building
graph {
 properties {bUsedWithNanite=true, pixelDepthOffsetMode=PDOM_Legacy, usedShadingModels=Default Lit}
 nodes [
  Constant() {r=0.2}
  Constant()
  Constant()
 ]
}
